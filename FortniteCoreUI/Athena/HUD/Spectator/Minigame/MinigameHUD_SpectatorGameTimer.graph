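# Widget Blueprint for the game timer shown in the minigame spectator HUD.
# A translucent dark border holds a horizontal row with the main time text and a smaller centiseconds text, each in a minimum-width size box.
# The centiseconds box is set to collapsed.
# The event graph is empty, so the Blueprint supplies only layout and styling for text widgets inherited from the native MinigameHUD_SpectatorGameTimer class.
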WidgetBlueprint /FortniteCoreUI/Athena/HUD/Spectator/Minigame/MinigameHUD_SpectatorGameTimer
graphs [EventGraph {links [], nodes []}]
variables []
widgets {
 info {
  inheritedWidgetCount=3
  parentClass=/Script/FortniteUI.MinigameHUD_SpectatorGameTimer
  rootWidgetClass=&CommonBorder
  widgetCount=7
 }
 widgets [
  CommonBorder_86 {
   widgetClassPath=&CommonBorder
   widget=CommonBorder_86
   properties {
    horizontalAlignment=HAlign_Center
    verticalAlignment=VAlign_Center
    background {
     bIsDynamicallyLoaded=false
     drawAs=Box
     imageType=NoImage
     mirroring=NoMirror
     resourceName=None
     resourceObject=null
     tiling=NoTile
     imageSize=<32 32>
     margin=<ltrb 0 0 0 0>
     outlineSettings {
      bUseBrushTransparency=false
      roundingType=HalfHeightRadius
      width=0
      color {colorUseRule=UseColor_Specified, specifiedColor=<rgba 0 0 0 0>}
      cornerRadii=<0 0 0 0>
     }
     tintColor {colorUseRule=UseColor_Specified, specifiedColor=<rgba 1 1 1 1>}
     uVRegion {bIsValid=false, max=<0 0>, min=<0 0>}
    }
    brushColor=<rgba 0 0 0 0.3>
    padding=<ltrb 5 0 5 0>
   }
  }
  HorizontalBox_0 {
   widgetClassPath=&HorizontalBox
   parent=CommonBorder_86
   widget=HorizontalBox_0
   slot {class=BorderSlot, properties {horizontalAlignment=HAlign_Center, verticalAlignment=VAlign_Center, padding=<ltrb 5 0 5 0>}}
  }
  SizeBox_0 {
   widgetClassPath=&SizeBox
   parent=HorizontalBox_0
   widget=SizeBox_0
   properties {bOverride_MinDesiredWidth=true, minDesiredWidth=107}
   slot {class=HorizontalBoxSlot}
  }
  Text_Time {
   widgetClassPath=&CommonTextBlock
   bInherited=true
   parent=SizeBox_0
   widget=Text_Time
   properties {
    justification=Center
    text=88:88
    font {
     bForceMonospaced=false
     bMaterialIsStencil=false
     fontMaterial=null
     fontObject=&BurbankSmall
     letterSpacing=0
     monospacedWidth=1
     size=28
     skewAmount=0
     typefaceFontName=Black
     outlineSettings {
      bApplyOutlineToDropShadows=false
      bMiteredCorners=false
      bSeparateFillAlpha=false
      outlineMaterial=null
      outlineSize=0
      outlineColor=<rgba 0 0 0 1>
     }
    }
    shadowOffset=<0 0>
   }
   slot {class=SizeBoxSlot, properties {horizontalAlignment=HAlign_Center, verticalAlignment=VAlign_Center}}
  }
  SizeBox_32 {
   widgetClassPath=&SizeBox
   parent=HorizontalBox_0
   widget=SizeBox_32
   properties {bOverride_MinDesiredWidth=true, minDesiredWidth=35, visibility=Collapsed}
   slot {class=HorizontalBoxSlot}
  }
  Text_Centiseconds {
   widgetClassPath=&CommonTextBlock
   bInherited=true
   parent=SizeBox_32
   widget=Text_Centiseconds
   properties {
    text=:88
    font {
     bForceMonospaced=false
     bMaterialIsStencil=false
     fontMaterial=null
     fontObject=&BurbankSmall
     letterSpacing=0
     monospacedWidth=1
     size=16
     skewAmount=0
     typefaceFontName=Black
     outlineSettings {
      bApplyOutlineToDropShadows=false
      bMiteredCorners=false
      bSeparateFillAlpha=false
      outlineMaterial=null
      outlineSize=0
      outlineColor=<rgba 0 0 0 1>
     }
    }
    shadowOffset=<0 0>
   }
   slot {class=SizeBoxSlot, properties {horizontalAlignment=HAlign_Left, verticalAlignment=VAlign_Center, padding=<ltrb 0 5 0 0>}}
  }
  TutorialHighlight {widgetClassPath=&TutorialHighlight}
 ]
}
